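; Burst gameplay cue for the rift POI teleport that plays the exit effects after a teleport completes.
; On burst, it reads the location from the cue parameters and spawns the rift Niagara system there with auto-release pooling.
; It also plays the exit-rift sound at that same location.
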(event EventOnBurst
  (params (Target :type Actor%20Object%20Reference) (Parameters :type Gameplay%20Cue%20Parameters%20Structure%20%28by%20ref%29) (SpawnResults :type Gameplay%20Cue%20Notify%20Spawn%20Result%20Structure%20%28by%20ref%29))
  (bind (_normalizedmagnitude _rawmagnitude _effectcontext _matchedtagname _originaltag _aggregatedsourcetags _aggregatedtargettags _location) (Ability|GameplayCue|BreakGameplayCueParameters Parameters))
  (Niagara|SpawnSystematLocation :WorldContextObject Target :SystemTemplate (Variables|Default|Getvfx_Rift) :Location _location :Scale "1.000000,1.000000,1.000000" :bAutoDestroy true :bAutoActivate true :PoolingMethod "AutoRelease" :bPreCullCheck false)
  (Audio|PlaySoundatLocation Target (Variables|Default|Getsfx_ExitRift) _location))

(fn ConstructionScript)
